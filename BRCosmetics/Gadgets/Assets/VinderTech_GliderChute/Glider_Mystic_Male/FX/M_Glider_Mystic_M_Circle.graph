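# Translucent unlit Niagara material for the magic circle effect on the male Mystic glider, with a static switch between sprite and mesh-style paths.
# Rotating, panning cloud noise distorts the dynamically rotated circle texture, whose channels are masked by vertex color, cut by noise and depth-faded.
# The sprite path instead uses a thresholded mask; the other adds a Z offset from vertex color and dynamic parameters.
# Time-of-day tint scales the emissive.
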Material /BRCosmetics/Gadgets/Assets/VinderTech_GliderChute/Glider_Mystic_Male/FX/M_Glider_Mystic_M_Circle
graph {
 properties {
  bUsedWithNiagaraMeshParticles=true
  bUsedWithNiagaraSprites=true
  blendMode=BLEND_Translucent
  pixelDepthOffsetMode=PDOM_Legacy
  shadingModel=MSM_Unlit
  twoSided=true
  usedShadingModels=Unlit
 }
 nodes [
  n1 = DynamicParameter() {parameterIndex=1, defaultValue=<rgba 1 1 1 0.3>, paramNames [Intensity_R,Intensity_G,Intensity_B,Intensity_B2]}
  n2 = TextureSample() {
   samplerType=SAMPLERTYPE_LinearColor
   texture=&/T_Mystic_Magic_Circle_mask
  }
  n3 = ParticleColor()
  n4 = Multiply(A=Power(Base=Multiply(A=@n1.Intensity_B2, B=@n2.B), Exp=ScalarParameter() {defaultValue=2, parameterName=Threshold}), B=@n3.RGB)
  n5 = TextureCoordinate()
  n6 = DynamicParameter() {parameterIndex=2, defaultValue=<rgba 0 1 0 0>, paramNames [UVRot,G_Threshold,3,4]}
  n7 = Subtract(A=TextureCoordinate() {uTiling=6, vTiling=6}) {constB=3}
  n8 = Time()
  n9 = Multiply(A=@n8) {constB=-1}
  n10 = Multiply(A=Constant2Vector() {g=-0.3, r=0.45}, B=ScalarParameter() {defaultValue=1, parameterName=PanSpeed})
  n11 = TextureSample(UVs=Add(A=Rotator(Coordinate=@n7, Time=@n9), B=Multiply(A=@n10, B=@n8))) {samplerType=SAMPLERTYPE_Masks, texture=/Game/Global/Textures/Default/Noise/T_Noise_Cloud_Mask}
  n12 = Multiply(A=Subtract(A=@n11.RGB) {constB=0.5}, B=ScalarParameter() {defaultValue=0.004, parameterName=UV_Distort})
  n13 = TextureSampleParameter2D(UVs=Add(A=Rotator(Coordinate=@n5, Time=@n6.UVRot), B=ComponentMask(None=@n12) {g=true, r=true})) {
   parameterName=Texture
   samplerType=SAMPLERTYPE_LinearColor
   texture=&/T_Mystic_Magic_Circle_0
  }
  n14 = MaterialFunctionCall(Float3=Multiply(
    A=MaterialFunctionCall(X=@n1.Intensity_R, Y=@n1.Intensity_R, Z=@n1.Intensity_R) {materialFunction=&MakeFloat3}.Result
    B=@n13.RGB
   )) {materialFunction=&BreakOutFloat3Components}
  n15 = VertexColor()
  n16 = Add(A=Multiply(A=@n14.R, B=@n15.R), B=Multiply(A=@n14.G, B=Ceil(None=@n15.G)))
  n17 = DynamicParameter() {defaultValue=<rgba 3 1 0 0>, paramNames [Ofs_R,2,3,4]}
  n18 = Add(A=@n2.R, B=@n2.R)
  n19 = Ceil(None=Subtract(A=Multiply(A=@n17.4, B=@n18)) {constB=0.5})
  n20 = Multiply(A=Saturate(None=Subtract(A=@n14.B, B=@n19)), B=@n15.B)
  n21 = MaterialFunctionCall(Float3=Multiply(A=@n11.RGB, B=VectorParameter() {parameterName=Noises, defaultValue=<rgba 0.4 0.5 1 0>}.RGB)) {materialFunction=&BreakOutFloat3Components}
  n22 = Multiply(A=Multiply(A=@n3.A, B=Add(A=@n16, B=@n20)), B=Subtract(B=@n21.G))
  n23 = StaticBoolParameter() {parameterName=IS_Sprite}
  DepthFade(
   Opacity=Saturate(None=StaticSwitch(True=@n4, False=@n22, Value=@n23))
   FadeDistance=ScalarParameter() {defaultValue=60, parameterName=DepthFade}
  )
  n24 = MaterialFunctionCall(
   Color_Morning=VectorParameter() {group=TimeOfDay, parameterName=ColorTint_Morning, defaultValue=<rgba 1 1 1 1>}.RGB
   Color_Day=VectorParameter() {group=TimeOfDay, parameterName=ColorTint_Day, defaultValue=<rgba 1 1 1 1>}.RGB
   Color_Evening=VectorParameter() {group=TimeOfDay, parameterName=ColorTint_Evening, defaultValue=<rgba 1 1 1 1>}.RGB
   Color_Night=VectorParameter() {group=TimeOfDay, parameterName=ColorTint_Night, defaultValue=<rgba 1 1 1 1>}.RGB
  ) {materialFunction=&Color_Time_of_Day}
  n25 = Multiply(A=VectorParameter() {parameterName=ColorAdjustment, defaultValue=<rgba 4 1.948 0 0>}.RGB, B=@n4)
  n26 = Subtract(A=@n3.RGB, B=Multiply(A=@n21.R, B=@n21.R))
  MaterialFunctionCall(Emissive=Multiply(A=VertexInterpolator(VS=@n24.Result).PS, B=StaticSwitch(True=@n25, False=@n26, Value=@n23))) {materialFunction=&MF_FX_EmissiveAdjustments}
  n27 = Constant()
  n28 = MaterialFunctionCall(Float3=Multiply(
    A=MaterialFunctionCall(X=@n17.Ofs_R, Y=@n17.Ofs_R, Z=@n17.Ofs_R) {materialFunction=&MakeFloat3}.Result
    B=@n15
   )) {materialFunction=&BreakOutFloat3Components}
  n29 = Add(A=@n28.R, B=@n28.R)
  StaticSwitch(
   True=@n27
   False=MaterialFunctionCall(X=@n27, Y=@n27, Z=Add(A=@n29, B=@n28.B)) {materialFunction=&MakeFloat3}.Result
   Value=@n23
  )
 ]
}
