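# Translucent, unlit, two-sided Niagara mesh-particle material that draws the sparks circle effect for the Mystic male glider.
# The sparks circle texture's U axis is offset by the UVRot dynamic parameter, so particles can drive the apparent rotation, then tiled by U_repeat, negative by default.
# Particle color tints the texture into an emissive adjustment function, while particle alpha times the texture forms a second product, apparently for opacity.
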Material /BRCosmetics/Gadgets/Assets/VinderTech_GliderChute/Glider_Mystic_Male/FX/M_Glider_Mystic_SparksCircle
graph {
 properties {
  bCastRayTracedShadows=false
  bUsedWithNiagaraMeshParticles=true
  blendMode=BLEND_Translucent
  pixelDepthOffsetMode=PDOM_Legacy
  shadingModel=MSM_Unlit
  twoSided=true
  usedShadingModels=Unlit
 }
 nodes [
  n1 = ParticleColor()
  n2 = TextureCoordinate()
  n3 = AppendVector(A=DynamicParameter() {parameterIndex=2, defaultValue=<rgba 0 1 0 0>, paramNames [UVRot,G_Threshold,3,4]}.UVRot, B=Constant())
  n4 = ScalarParameter() {defaultValue=-2, parameterName=U_repeat}
  n5 = Constant() {r=1}
  n6 = TextureSample(UVs=Multiply(A=Add(A=@n2, B=@n3), B=AppendVector(A=@n4, B=@n5))) {texture=&/T_Mystic_Magic_SparksCircle}
  MaterialFunctionCall(Emissive=Multiply(A=@n1.RGB, B=@n6.RGB)) {materialFunction=&MF_FX_EmissiveAdjustments}
  Multiply(A=@n1.A, B=@n6.RGB)
  n7 = VertexColor()
  OneMinus(None=Multiply(A=@n7.R, B=@n7.R))
  Rotator(Coordinate=TextureCoordinate(), Time=Constant() {r=-0.666})
  Time()
 ]
}
