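# Additive unlit material for spark effects on Niagara ribbons and sprites, filed under the Mystic Male glider's FX.
# A spark mask, with parameter-blended UV axis swapping, multiplies a tiled, time-panned ribbon sparks texture and particle color.
# The graph includes emissive adjustments using particle alpha and a low-quality branch that drops the spark mask.
# A static switch selects near-camera and depth fading, with ES2 and ES3.1 skipping depth fade.
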Material /BRCosmetics/Gadgets/Assets/VinderTech_GliderChute/Glider_Mystic_Male/FX/M_InitialRIbbonSparks
graph {
 properties {
  allowTranslucentCustomDepthWrites=true
  bEnableResponsiveAA=true
  bUsedWithNiagaraRibbons=true
  bUsedWithNiagaraSprites=true
  blendMode=BLEND_Additive
  opacityMaskClipValue=0.01
  pixelDepthOffsetMode=PDOM_Legacy
  shadingModel=MSM_Unlit
  usedShadingModels=Unlit
 }
 nodes [
  n1 = LinearInterpolate(
   A=TextureCoordinate()
   B=MaterialFunctionCall(XY=TextureCoordinate()) {materialFunction=&Swizzle}.YX
   Alpha=ScalarParameter() {parameterName=SwizzleUV}
  )
  n2 = Time()
  n3 = Constant2Vector() {g=3, r=2}
  n4 = TextureSample(UVs=Add(A=TextureCoordinate() {uTiling=2.75}, B=Multiply(A=@n2, B=@n3))) {samplerType=SAMPLERTYPE_LinearColor, texture=&/T_Mystic_Ribbon_Sparks}
  n5 = ParticleColor()
  n6 = Multiply(
   A=Multiply(A=TextureSampleParameter2D(UVs=@n1) {parameterName=Texture, texture=/Game/Weapons/FORT_Rifles/Effects/Textures/spark}.R, B=@n4.RGB)
   B=@n5.RGB
  )
  MaterialFunctionCall(Emissive=Multiply(A=@n6, B=@n5.A)) {materialFunction=&MF_FX_EmissiveAdjustments}
  n7 = Constant() {r=1.6}
  QualitySwitch(Default=@n6, Low=Multiply(A=Multiply(A=@n7, B=@n4.RGB), B=@n5.RGB))
  n8 = MaterialFunctionCall(Distance=ScalarParameter() {defaultValue=200, parameterName=Near Camera Distance}) {materialFunction=&Near_Camera_Fade}
  n9 = DepthFade(Opacity=@n8.Result, FadeDistance=ScalarParameter() {defaultValue=50, parameterName=Depth Fade Distance})
  StaticSwitchParameter(True=FeatureLevelSwitch(Default=@n9, ES2=@n8.Result, ES3_1=@n8.Result, SM4_REMOVED=@n9), False=Constant() {r=1}) {parameterName="Use Near Camera Fade && Depth Fade"}
 ]
}
