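# Niagara idle effect for the male Mystic glider: a fiery orange magic circle that builds up in timed stages.
# Ribbon and spark intros lead into glows, layered ring meshes and looping GPU sparks, mostly positioned on User.SkeletalMesh.
# Shared Emitter.Age curves grow the circle in two steps, with animated dynamic material parameters.
# User.ForwardDot drives main spark rate, swirl and drag; FN_FrontEndFloatModifier switches or retunes elements between gameplay and front end.
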NiagaraSystem /BRCosmetics/Gadgets/Assets/VinderTech_GliderChute/Glider_Mystic_Male/FX/NS_Glider_Mystic_M_Idle
dependencies=null
emitters {
 Center_Glow {
  emitterUpdateScript [
   &EmitterState(
    Enable Distance Culling=true
    Inactive Response=::NewEnumerator1
    Life Cycle Mode=::NewEnumerator1
    Loop Behavior=::NewEnumerator1
    Loop Duration Mode=::NewEnumerator1
    MaxDistance=10000
    Module.bUseMaxDistance=true
    Module.bUseMinDistance=false
    Scalability Mode=::NewEnumerator1
    UseLoopDelay=false
   )
   &SpawnBurst_Instantaneous(Use Spawn Probability=false, UseLoopCountLimit=false)
  ]
  particleSpawnScript [&InitializeParticle(
    Color=<rgba 2.6 0.31 0 1>
    Lifetime Max=1
    Lifetime Min=0.8
    Lifetime Mode=::NewEnumerator1
    Sprite Size=<500 400>
    Sprite Size Mode=::NewEnumerator4
    UsePositionOffset=true
    Write Material Random=true
   )]
  particleUpdateScript [
   &ParticleState(Kill Particles When Lifetime Has Elapsed=false, Let Infinitely Lived Particles Die When Emitter Deactivates=true)
   &SkeletalMeshLocation(
    Bone Sampling Mode=::NewEnumerator4
    "Interpolate Mesh Sampling (Update)"=true
    MaskSpawnGroup=false
    Sampled Orientation Coordinate Space=::Local
    Sampled Position Offset=<0 0 3>
    Sampled Position Offset Coordinate Space:ENiagaraSkelMeshTransforms=::NewEnumerator0
    Skeletal Mesh=$User.SkeletalMesh
    Use Sampled Mesh Attribute as Ribbon ID=false
   )
   &ScaleColor(
    Scale Alpha=&FloatFromCurve(
     FloatCurve=DataInterface() {propertyValues {Curve {keys [
         {arriveTangent=3.6277, interpMode=RCIM_Cubic, leaveTangent=3.6277, tangentMode=RCTM_User, time=0, value=0}
         {interpMode=RCIM_Cubic, tangentMode=RCTM_User, time=0.5, value=1}
        ]}}}
     CurveIndex=$Emitter.Age
     Scale Curve=&FloatFromCurve(
      FloatCurve=DataInterface() {propertyValues {Curve {keys [
          {interpMode=RCIM_Cubic, tangentMode=RCTM_User, time=1.1, value=0}
          <at 1.4 0.25>
          {arriveTangent=2.7482, interpMode=RCIM_Cubic, leaveTangent=2.7482, tangentMode=RCTM_User, time=1.9, value=0.25}
          {interpMode=RCIM_Cubic, tangentMode=RCTM_User, time=2.4, value=1}
         ]}}}
      CurveIndex=$Emitter.Age
      Scale Curve=0.3
     )
    )
    ScaleA=true
    ScaleRGB=false
   )
   &ScaleSpriteSize(
    Uniform Curve Index=$Emitter.Age
    Uniform Curve Sprite Scale=DataInterface() {propertyValues {Curve {keys [
        {interpMode=RCIM_Cubic, tangentMode=RCTM_User, time=1.1, value=0}
        <at 1.4 0.25>
        {arriveTangent=2.7482, interpMode=RCIM_Cubic, leaveTangent=2.7482, tangentMode=RCTM_User, time=1.9, value=0.25}
        {interpMode=RCIM_Cubic, tangentMode=RCTM_User, time=2.4, value=1}
       ]}}}
   )
   &CameraOffset(Camera Offset Amount=-20)
  ]
  rendererClasses [&NiagaraSpriteRendererProperties]
  renderers [{rendererClass=&NiagaraSpriteRendererProperties}]
 }
 FN_Ribbon_Attributes_Test {
  emitterUpdateScript [
   &EmitterState(
    Enable Distance Culling=true
    Life Cycle Mode=::NewEnumerator1
    Loop Behavior=::NewEnumerator1
    Loop Delay=0.2542
    Loop Duration=&FN_FrontEndFloatModifier(Gameplay=1.475, Front End=1.1)
    MaxDistance=7000
    Module.bUseMaxDistance=true
    Module.bUseMinDistance=false
    Scalability Mode=::NewEnumerator1
    UseLoopDelay=true
   ) {module=EmitterLifeCycle}
   &SpawnRate(SpawnRate=60, Use Spawn Probability=false)
  ]
  particleSpawnScript [
   /SetVariables(Particles.My_Random:NiagaraFloat=&UniformRangedFloat(Minimum=0, Maximum=1)) {isSetParameters=true}
   /SetVariables(Particles.ParticleId_Thing=HlslExpression() {hlslExpression=Particles.ID}) {isSetParameters=true}
   &SkeletalMeshLocation(
    Bone Sampling Mode=::NewEnumerator4
    MaskSpawnGroup=false
    Orientation Sampling=::NewEnumerator1
    Sampled Position Offset Coordinate Space:ENiagaraSkelMeshTransforms=::NewEnumerator0
    Skeletal Mesh=$User.SkeletalMesh
    Use Sampled Mesh Attribute as Ribbon ID=false
   ) {module=SkeletalMeshLocation001}
   &InitializeRibbon(
    Lifetime=0.5
    Mass=10
    Module.Write Color=true
    Module.Write Lifetime=true
    Module.Write Mass=true
    Module.Write Material Random=true
    Module.Write Position=false
    Module.Write RibbonTwist=false
    Module.Write RibbonWidth=true
    Position:Vector3f=&SimulationPosition()
    Ribbon Width=2
   )
   &TorusLocation(
    Align Torus With Axis=true
    Large Radius=20
    MaskSpawnGroup=false
    Offset=<0 0 -4>
    Torus Axis=<0 0 0>
    Torus Distribution Mode:ENiagaraTorusDistributionMode=::NewEnumerator1
    Torus Mode=::NewEnumerator2
    U Position:NiagaraFloat=&Multiply_Float(A=$Emitter.Age, B=-2)
   )
  ]
  particleUpdateScript [
   &ParticleState()
   /SetVariables(Particles.RibbonWidth=&Cosine(Normalized Angle=HlslExpression() {hlslExpression=Particles.Age/2}, Period=1, Scale=2.05, Bias=5)) {isSetParameters=true}
   &Color(Color=&MultiplyLinearColor(
     Linear Color=&ColorFromCurve(ColorCurve=DataInterface() {propertyValues {
        AlphaCurve {keys [
          <at 0.0062 0>
          <at 0.3181 0.6>
          {arriveTangent=-0.8799, interpMode=RCIM_Cubic, leaveTangent=-0.8799, time=0.5278, value=0.5}
          <at 1 0>
          <at 1 0>
         ]}
        BlueCurve {keys [<at 0 0 RCIM_Cubic>,<at 0.9977 0 RCIM_Cubic>]}
        GreenCurve {keys [<at 0 9.8361 RCIM_Cubic>,<at 0.9977 0.1345 RCIM_Cubic>]}
        RedCurve {keys [<at 0 128 RCIM_Cubic>,<at 0.9977 2 RCIM_Cubic>]}
       }})
     Scale Factor=<1 1 1 1>
    )) {module=Color003}
   &SolveForcesAndVelocity(Clamp Velocity=false, Limit Acceleration=false)
  ]
  rendererClasses [&NiagaraRibbonRendererProperties]
  renderers [{rendererClass=&NiagaraRibbonRendererProperties}]
 }
 Feet_Glow {
  emitterUpdateScript [
   &EmitterState(
    Enable Distance Culling=true
    Inactive Response=::NewEnumerator1
    Life Cycle Mode=::NewEnumerator1
    Loop Behavior=::NewEnumerator1
    Loop Duration Mode=::NewEnumerator1
    MaxDistance=7000
    Module.bUseMaxDistance=true
    Module.bUseMinDistance=false
    Scalability Mode=::NewEnumerator1
    UseLoopDelay=false
   )
   &SpawnBurst_Instantaneous(Spawn Time=0.01, Use Spawn Probability=false, UseLoopCountLimit=false)
   &SpawnBurst_Instantaneous(Spawn Time=0.05, Use Spawn Probability=false, UseLoopCountLimit=false) {module=SpawnBurst_Instantaneous001}
  ]
  particleSpawnScript [&InitializeParticle(
    Color=<rgba 2.6 0.31 0 1.4>
    Lifetime Max=&FN_FrontEndFloatModifier(Gameplay=1, Front End=0)
    Lifetime Min:NiagaraFloat=&FN_FrontEndFloatModifier(Gameplay=0.8, Front End=0)
    Lifetime Mode=::NewEnumerator1
    Position Offset=&Multiply_VectorByFloat(Vector=<0 64 -5>, Float=&Add_Float(A=HlslExpression() {hlslExpression=Particles.ID}, B=-0.5))
    Sprite Size=<65 38>
    Sprite Size Mode=::NewEnumerator4
    UsePositionOffset=true
    Write Material Random=true
   )]
  particleUpdateScript [
   &ParticleState(Kill Particles When Lifetime Has Elapsed=false, Let Infinitely Lived Particles Die When Emitter Deactivates=true)
   &SkeletalMeshLocation(
    Bone Sampling Mode=::NewEnumerator4
    MaskSpawnGroup=false
    Orientation Sampling=::NewEnumerator1
    Sampled Position Offset=<0 0 -1>
    Sampled Position Offset Coordinate Space:ENiagaraSkelMeshTransforms=::NewEnumerator2
    Skeletal Mesh=$User.SkeletalMesh
    Use Sampled Mesh Attribute as Ribbon ID=false
   ) {enabled=false}
   &ScaleColor(
    Scale Alpha=&FloatFromCurve(
     FloatCurve=DataInterface() {propertyValues {Curve {keys [
         {arriveTangent=-0.0713, interpMode=RCIM_Cubic, leaveTangent=-0.0713, tangentMode=RCTM_User, time=2.1, value=0}
         {interpMode=RCIM_Cubic, tangentMode=RCTM_User, time=3, value=1}
        ]}}}
     CurveIndex=$Emitter.Age
     Scale Curve=&FloatFromCurve(
      FloatCurve=DataInterface() {propertyValues {Curve {keys [
          {interpMode=RCIM_Cubic, tangentMode=RCTM_User, time=1.1, value=0}
          <at 1.4 0.25>
          {arriveTangent=2.7482, interpMode=RCIM_Cubic, leaveTangent=2.7482, tangentMode=RCTM_User, time=1.9, value=0.25}
          {interpMode=RCIM_Cubic, tangentMode=RCTM_User, time=2.4, value=1}
         ]}}}
      CurveIndex=$Emitter.Age
      Scale Curve=&FN_FrontEndFloatModifier(Gameplay=1, Front End=0)
     )
    )
    ScaleA=true
    ScaleRGB=false
   )
   &ScaleSpriteSize(
    Uniform Curve Index=$Emitter.Age
    Uniform Curve Sprite Scale=DataInterface() {propertyValues {Curve {keys [
        {interpMode=RCIM_Cubic, tangentMode=RCTM_User, time=1.1, value=0}
        <at 1.4 0.25>
        {arriveTangent=2.7482, interpMode=RCIM_Cubic, leaveTangent=2.7482, tangentMode=RCTM_User, time=1.9, value=0.25}
        {interpMode=RCIM_Cubic, tangentMode=RCTM_User, time=2.4, value=1}
       ]}}}
   )
   &CameraOffset(Camera Offset Amount=16)
  ]
  rendererClasses [&NiagaraSpriteRendererProperties]
  renderers [{rendererClass=&NiagaraSpriteRendererProperties}]
 }
 First_Sparks {
  simTarget=GPUComputeSim
  emitterSpawnScript [/SetVariables(Emitter.Position:Vector3f=&Add_Vector(A=/Niagara/DynamicInputs/Helpers/SimulationPosition(), B=<10 0 20>)) {isSetParameters=true}]
  emitterUpdateScript [
   &EmitterState(
    Enable Distance Culling=true
    Life Cycle Mode=::NewEnumerator1
    Loop Behavior=::NewEnumerator1
    Loop Delay=0.4417
    Loop Duration=1
    MaxDistance=7000
    Module.bUseMaxDistance=true
    Module.bUseMinDistance=false
    Scalability Mode=::NewEnumerator1
    UseLoopDelay=true
   )
   &SpawnBurst_Instantaneous(Spawn Count=20, Use Spawn Probability=false, UseLoopCountLimit=false)
   &SpawnBurst_Instantaneous(Spawn Count=20, Spawn Time=0.225, Use Spawn Probability=false, UseLoopCountLimit=false) {module=SpawnBurst_Instantaneous001}
   &SpawnBurst_Instantaneous(Spawn Count=20, Spawn Time=0.3167, Use Spawn Probability=false, UseLoopCountLimit=false) {module=SpawnBurst_Instantaneous002}
  ]
  particleSpawnScript [
   &InitializeParticle(
    Color=<rgba 5 0.9789 0 1>
    Lifetime Max=0.3
    Lifetime Min=0.2
    Lifetime Mode=::NewEnumerator1
    Position Offset Coordinate Space=::World
    Sprite Size Max=<3 30>
    Sprite Size Min=<1.5 50>
    Sprite Size Mode=::NewEnumerator2
    UsePositionOffset=true
    Write Material Random=true
   )
   &SkeletalMeshLocation(
    Bone Sampling Mode=::NewEnumerator4
    MaskSpawnGroup=false
    Orientation Sampling=::NewEnumerator1
    Sampled Position Offset Coordinate Space:ENiagaraSkelMeshTransforms=::NewEnumerator2
    Skeletal Mesh=$User.SkeletalMesh
    Use Sampled Mesh Attribute as Ribbon ID=false
    Write To Particle Position=false
   )
   &AddVelocityFromPoint(Module.Use Falloff Distance=false, Velocity Strength=&RandomRangeFloat(Minimum=100, Maximum=400))
  ]
  particleUpdateScript [
   &ParticleState()
   &Drag(Drag=3, Use Linear Drag=true, Use Rotational Drag=false)
   &CurlNoiseForce(Noise Frequency=20, Noise Strength=500, Pan Noise=false)
   &SolveForcesAndVelocity(Clamp Velocity=false, Limit Acceleration=false)
   &ScaleSpriteSize(
    Scale Factor=&Vector2DFromCurve(
     Vector2Curve=DataInterface() {propertyValues {
       XCurve {keys [<at 0 1>,<at 1 0.4>]}
       YCurve {keys [{arriveTangent=-3.4038, interpMode=RCIM_Cubic, leaveTangent=-3.4038, tangentMode=RCTM_User, time=0, value=1}, <at 1 0 RCIM_Cubic>]}
      }}
     ScaleCurve=<1 1>
    )
    Scale Sprite Size Mode=::NewEnumerator2
   )
   &JitterPosition(Jitter Amount=2)
  ]
  rendererClasses [&NiagaraSpriteRendererProperties]
  renderers [{rendererClass=&NiagaraSpriteRendererProperties}]
 }
 GPU_Sparks_Frontend {
  simTarget=GPUComputeSim
  emitterUpdateScript [
   &EmitterState(
    Delay First Loop Only=true
    Life Cycle Mode=::NewEnumerator1
    Loop Delay=2.3833
    Loop Duration=1
    Recalculate Duration Each Loop=true
    UseLoopDelay=true
   )
   &SpawnRate(
    Spawn Probability=&FN_FrontEndFloatModifier(Gameplay=0, Front End=1)
    SpawnRate=&Lerp_Float(A=800, B=2000, Alpha=$User.ForwardDot)
    Use Spawn Probability=true
   )
  ]
  particleSpawnScript [
   &InitializeParticle(
    Color=<rgba 10 0.98 0.05 1>
    Lifetime Max=0.34
    Lifetime Min=0.2
    Lifetime Mode=::NewEnumerator1
    Sprite Rotation Mode=::NewEnumerator2
    Sprite Size Mode=::NewEnumerator1
    Uniform Sprite Size Max=7
    UsePositionOffset=false
    Write Material Random=true
   )
   &SkeletalMeshLocation(
    Bone Sampling Mode=::NewEnumerator4
    MaskSpawnGroup=false
    Sampled Position Offset=<0 0 5>
    Sampled Position Offset Coordinate Space:ENiagaraSkelMeshTransforms=::NewEnumerator2
    Sampled Velocity Coordinate Space:ENiagaraCoordinateSpace=::Local
    Skeletal Mesh=$User.SkeletalMesh
   )
   &ShapeLocation(
    Handle Radius=1
    Large Radius=98
    MaskSpawnGroup=false
    Offset=<0 0 -4.5>
    Offset Coordinate Space=::World
    Offset Mode=::NewEnumerator0
    Override Seed=false
    Rotation Axis=&RotateVectorByQuaternion(Vector To Rotate=<0 0 1>, Quaternion=$Particles.MeshOrientation)
    Rotation Mode=::NewEnumerator1
    Shape Primitive=::NewEnumerator3
    UseOwnerScale=true
   )
   /Niagara/Modules/Update/Velocity/VortexVelocity(
    Module.Use Influence Falloff=false
    Velocity Amount:NiagaraFloat=&Lerp_Float(A=-300, B=&FN_FrontEndFloatModifier(Gameplay=-450, Front End=-300), Alpha=$User.ForwardDot)
   )
   &ApplyInitialForces()
  ]
  particleUpdateScript [
   &ParticleState()
   &ScaleColor(
    Scale Alpha=&FloatFromCurve(FloatCurve=DataInterface() {propertyValues {Curve {keys [<at -0.0046 1.0113 RCIM_Cubic>,<at 1.0011 0.002 RCIM_Cubic>]}}})
    Scale RGB=&VectorFromCurve(VectorCurve=DataInterface() {propertyValues {XCurve {keys [<at 0.5 1>,<at 1 2>]}, YCurve {keys [<at 0.4 1>,<at 1 0.25>]}, ZCurve {keys [<at 0.4 1>,<at 1 0.25>]}}})
    ScaleA=true
    ScaleRGB=false
   )
   &CurlNoiseForce(Noise Frequency=70, Noise Quality / Cost=::NewEnumerator0, Noise Strength=2000, Pan Noise=true, Pan Noise Field=<1.5 -0.2 0.5>)
   &Drag(Drag=&Lerp_Float(A=1, B=0.1, Alpha=$User.ForwardDot), Use Linear Drag=true, Use Rotational Drag=false)
   &InheritSourceMovement(
    Applied Position Scale=&RandomRangeVector(Minimum=<0 0 0>, Maximum=<-0.1 -0.1 -0.2>)
    ApplyForce=false
    ApplyPosition=true
    ApplyPositionFalloff=false
   )
   &SolveForcesAndVelocity(Clamp Velocity=false, Limit Acceleration=false)
   &ScaleSpriteSizeBySpeed(Max Scale Factor=<0.2 5>, Min Scale Factor=<0.15 2>)
   &SolveRotationalForcesAndVelocity()
  ]
  rendererClasses [&NiagaraSpriteRendererProperties]
  renderers [{rendererClass=&NiagaraSpriteRendererProperties}]
 }
 GPU_Sparks_Init {
  simTarget=GPUComputeSim
  emitterUpdateScript [
   &EmitterState(
    Enable Distance Culling=true
    Life Cycle Mode=::NewEnumerator1
    Loop Behavior=::NewEnumerator1
    Loop Delay=0.6833
    Loop Duration=&FN_FrontEndFloatModifier(Gameplay=2, Front End=1.2)
    MaxDistance=7000
    Module.bUseMaxDistance=true
    Module.bUseMinDistance=false
    Scalability Mode=::NewEnumerator1
    UseLoopDelay=true
   )
   &SpawnRate(
    SpawnRate=&FloatFromCurve(
     FloatCurve=DataInterface() {propertyValues {Curve {keys [
         {interpMode=RCIM_Cubic, tangentMode=RCTM_User, time=0, value=0}
         {arriveTangent=1.7761, interpMode=RCIM_Cubic, leaveTangent=1.7761, tangentMode=RCTM_User, time=1, value=1}
        ]}}}
     CurveIndex=$Emitter.Age
     Scale Curve=1500
    )
    Use Spawn Probability=false
   )
  ]
  particleSpawnScript [
   &InitializeParticle(
    Color=<rgba 12 1.176 0.06 1>
    Lifetime Max=0.22
    Lifetime Min=0.15
    Lifetime Mode=::NewEnumerator1
    Sprite Rotation Mode=::NewEnumerator2
    Sprite Size Mode=::NewEnumerator1
    Uniform Sprite Size Max=4
    Uniform Sprite Size Min=3
    UsePositionOffset=false
    Write Material Random=true
   )
   &SkeletalMeshLocation(
    Bone Sampling Mode=::NewEnumerator4
    MaskSpawnGroup=false
    Orientation Sampling=::NewEnumerator1
    Sampled Position Offset=<0 0 2.5>
    Sampled Position Offset Coordinate Space:ENiagaraSkelMeshTransforms=::NewEnumerator2
    Skeletal Mesh=$User.SkeletalMesh
    Use Sampled Mesh Attribute as Ribbon ID=false
   )
   &ShapeLocation(
    Handle Radius=1
    Large Radius=21.5
    MaskSpawnGroup=false
    Offset=<0 0 -5>
    Offset Mode=::NewEnumerator0
    Override Seed=false
    Shape Primitive=::NewEnumerator3
    UseOwnerScale=true
   )
   &VortexVelocity(Module.Use Influence Falloff=false, Velocity Amount=-200)
   &ApplyInitialForces()
  ]
  particleUpdateScript [
   &ParticleState()
   &ScaleColor(
    Scale Alpha=&FloatFromCurve(FloatCurve=DataInterface() {propertyValues {Curve {keys [<at -0.0046 1.0113 RCIM_Cubic>,<at 1.0011 0.002 RCIM_Cubic>]}}})
    ScaleA=true
    ScaleRGB=true
   )
   &CurlNoiseForce(Noise Frequency=70, Noise Quality / Cost=::NewEnumerator0, Noise Strength=1000, Pan Noise=true, Pan Noise Field=<1 -0.2 0.5>)
   &Drag(Drag=3, Use Linear Drag=true, Use Rotational Drag=false)
   &SolveForcesAndVelocity(Clamp Velocity=false, Limit Acceleration=false)
   &ScaleSpriteSizeBySpeed(Max Scale Factor=<0.2 4>, Min Scale Factor=<0.15 1>)
  ]
  rendererClasses [&NiagaraSpriteRendererProperties]
  renderers [{rendererClass=&NiagaraSpriteRendererProperties}]
 }
 GPU_Sparks_Tail {
  simTarget=GPUComputeSim
  emitterUpdateScript [
   &EmitterState(
    Delay First Loop Only=true
    Enable Distance Culling=true
    Life Cycle Mode=::NewEnumerator1
    Loop Delay=2.3833
    Loop Duration=1
    MaxDistance=7000
    Module.bUseMaxDistance=true
    Module.bUseMinDistance=false
    Recalculate Duration Each Loop=true
    Scalability Mode=::NewEnumerator1
    UseLoopDelay=true
   )
   &SpawnRate(SpawnRate=500, Use Spawn Probability=false)
  ]
  particleSpawnScript [
   &InitializeParticle(
    Color=&RandomRangeLinearColor(Minimum=<rgba 10 0.98 0.05 1>, Maximum=<rgba 10 0.98 0.05 0.5>)
    Lifetime Max=0.25
    Lifetime Min=0.2
    Lifetime Mode=::NewEnumerator1
    Sprite Rotation Mode=::NewEnumerator2
    Sprite Size Mode=::NewEnumerator1
    Uniform Sprite Size Min=4
    UsePositionOffset=false
    Write Material Random=true
   )
   &SkeletalMeshLocation(
    Bone Sampling Mode=::NewEnumerator4
    MaskSpawnGroup=false
    Sampled Position Offset=<0 0 4>
    Sampled Position Offset Coordinate Space:ENiagaraSkelMeshTransforms=::NewEnumerator2
    Skeletal Mesh=$User.SkeletalMesh
    Use Sampled Mesh Attribute as Ribbon ID=false
   )
   &ShapeLocation(
    Disc Coverage=0.1
    MaskSpawnGroup=false
    Offset=&RandomRangeVector(Minimum=<0 0 -8>, Maximum=<0 0 -4>)
    Offset Mode=::NewEnumerator0
    Override Seed=false
    Ring Radius=96
    Rotation Angle=0.5
    Rotation Axis=<0 0 1>
    Rotation Mode=::NewEnumerator1
    Shape Primitive=::NewEnumerator4
    U Distribution=0.45
    UseOwnerScale=true
   ) {enabled=false}
   &ShapeLocation(
    Handle Radius=1
    MaskSpawnGroup=false
    Offset=<0 0 -4.5>
    Offset Coordinate Space=::World
    Offset Mode=::NewEnumerator0
    Override Seed=false
    Rotation Angle=0.5
    Rotation Axis=&RotateVectorByQuaternion(Vector To Rotate=<0 0 1>, Quaternion=$Particles.MeshOrientation)
    Rotation Mode=::NewEnumerator1
    Shape Primitive=::NewEnumerator3
    U Distribution=0.45
    UseOwnerScale=true
   ) {module=ShapeLocation001}
   &ApplyInitialForces()
   &AddVelocity(Velocity=<0 500 0>)
  ]
  particleUpdateScript [
   &ParticleState()
   &ScaleColor(
    Scale Alpha=&FloatFromCurve(FloatCurve=DataInterface() {propertyValues {Curve {keys [<at -0.0046 1.0113 RCIM_Cubic>,<at 1.0011 0.002 RCIM_Cubic>]}}})
    ScaleA=true
    ScaleRGB=true
   )
   &CurlNoiseForce(Noise Frequency=70, Noise Quality / Cost=::NewEnumerator0, Noise Strength=600, Pan Noise=true, Pan Noise Field=<1.5 -0.2 3>)
   &PointAttractionForce(
    Attraction Radius=200
    Attraction Strength=50
    Attractor Position Offset:Vector3f=<0 1 0>
    Kill Within Radius=false
    Module.Use Falloff=true
    UseAttractorOffset=true
   )
   &Drag(Drag=2, Use Linear Drag=true, Use Rotational Drag=false)
   &InheritSourceMovement(
    Applied Position Scale=&RandomRangeVector(Minimum=<0 0 0>, Maximum=<-0.1 -0.1 -0.2>)
    ApplyForce=false
    ApplyPosition=true
    ApplyPositionFalloff=false
   )
   &SolveForcesAndVelocity(Clamp Velocity=false, Limit Acceleration=false)
   &SolveRotationalForcesAndVelocity()
   &ScaleSpriteSizeBySpeed(Max Scale Factor=<0.3 4>, Min Scale Factor=<0.2 1>)
  ]
  rendererClasses [&NiagaraSpriteRendererProperties]
  renderers [{rendererClass=&NiagaraSpriteRendererProperties}]
 }
 GPU_Sparks_gameplay_ {
  simTarget=GPUComputeSim
  emitterUpdateScript [
   &EmitterState(
    Delay First Loop Only=true
    Enable Distance Culling=true
    Life Cycle Mode=::NewEnumerator1
    Loop Delay=2.3833
    Loop Duration=1
    MaxDistance=7000
    Module.bUseMaxDistance=true
    Module.bUseMinDistance=false
    Recalculate Duration Each Loop=true
    Scalability Mode=::NewEnumerator1
    UseLoopDelay=true
   )
   &SpawnRate(
    Spawn Probability=&FN_FrontEndFloatModifier(Gameplay=1, Front End=0)
    SpawnRate=&Lerp_Float(A=800, B=2000, Alpha=$User.ForwardDot)
    Use Spawn Probability=true
   )
  ]
  particleSpawnScript [
   &InitializeParticle(
    Color=<rgba 10 0.98 0.05 1>
    Lifetime Max=0.34
    Lifetime Min=0.2
    Lifetime Mode=::NewEnumerator1
    Sprite Rotation Mode=::NewEnumerator2
    Sprite Size Mode=::NewEnumerator1
    Uniform Sprite Size Max=7
    UsePositionOffset=false
    Write Material Random=true
   )
   &SkeletalMeshLocation(
    Bone Sampling Mode=::NewEnumerator4
    MaskSpawnGroup=false
    Sampled Position Offset=<0 0 5>
    Sampled Position Offset Coordinate Space:ENiagaraSkelMeshTransforms=::NewEnumerator2
    Sampled Velocity Coordinate Space:ENiagaraCoordinateSpace=::Local
    Skeletal Mesh=$User.SkeletalMesh
   ) {enabled=false}
   &ShapeLocation(
    Handle Radius=1
    Large Radius=98
    MaskSpawnGroup=false
    Offset=<0 0 -4.5>
    Offset Coordinate Space=::World
    Offset Mode=::NewEnumerator0
    Override Seed=false
    Rotation Axis=&RotateVectorByQuaternion(Vector To Rotate=<0 0 1>, Quaternion=$Particles.MeshOrientation)
    Rotation Mode=::NewEnumerator1
    Shape Primitive=::NewEnumerator3
    UseOwnerScale=true
   )
   &VortexVelocity(
    Module.Use Influence Falloff=false
    Velocity Amount:NiagaraFloat=&Lerp_Float(A=-300, B=&FN_FrontEndFloatModifier(Gameplay=-450, Front End=-300), Alpha=$User.ForwardDot)
   )
   &ApplyInitialForces()
  ]
  particleUpdateScript [
   &ParticleState()
   &ScaleColor(
    Scale Alpha=&FloatFromCurve(FloatCurve=DataInterface() {propertyValues {Curve {keys [<at -0.0046 1.0113 RCIM_Cubic>,<at 1.0011 0.002 RCIM_Cubic>]}}})
    Scale RGB=&VectorFromCurve(VectorCurve=DataInterface() {propertyValues {XCurve {keys [<at 0.5 1>,<at 1 2>]}, YCurve {keys [<at 0.4 1>,<at 1 0.25>]}, ZCurve {keys [<at 0.4 1>,<at 1 0.25>]}}})
    ScaleA=true
    ScaleRGB=false
   )
   &CurlNoiseForce(Noise Frequency=70, Noise Quality / Cost=::NewEnumerator0, Noise Strength=2000, Pan Noise=true, Pan Noise Field=<1.5 -0.2 0.5>)
   &Drag(Drag=&Lerp_Float(A=1, B=0.1, Alpha=$User.ForwardDot), Use Linear Drag=true, Use Rotational Drag=false)
   &InheritSourceMovement(
    Applied Position Scale=&RandomRangeVector(Minimum=<0 0 0>, Maximum=<-0.1 -0.1 -0.2>)
    ApplyForce=false
    ApplyPosition=true
    ApplyPositionFalloff=false
   )
   &SolveForcesAndVelocity(Clamp Velocity=false, Limit Acceleration=false)
   &ScaleSpriteSizeBySpeed(Max Scale Factor=<0.2 5>, Min Scale Factor=<0.15 2>)
   &SolveRotationalForcesAndVelocity()
  ]
  rendererClasses [&NiagaraSpriteRendererProperties]
  renderers [{rendererClass=&NiagaraSpriteRendererProperties}]
 }
 LineSprite {
  emitterUpdateScript [
   &EmitterState(
    Delay First Loop Only=true
    Enable Distance Culling=true
    Life Cycle Mode=::NewEnumerator1
    Loop Delay=4
    Loop Duration=2
    MaxDistance=7000
    Module.bUseMaxDistance=true
    Module.bUseMinDistance=false
    Recalculate Duration Each Loop=true
    Scalability Mode=::NewEnumerator1
    UseLoopDelay=true
   )
   &SpawnRate(SpawnRate=&FN_FrontEndFloatModifier(Gameplay=24, Front End=16), Use Spawn Probability=false)
   &SampleSkeletalMesh(
    AdditionalYawPitchRoll=false
    Mesh Sampling Type:ENiagaraSkelSamplingModeFull=::NewEnumerator5
    Skeletal Mesh:NiagaraDataInterfaceSkeletalMesh=$User.SkeletalMesh
    Socket Sampling Mode:ENiagaraSocketSamplingMode=::NewEnumerator5
   )
  ]
  particleSpawnScript [
   &InitializeParticle(
    Color Channel Mode:ENiagaraChannelCorrelation=::NewEnumerator2
    Color Maximum=<rgba 4 0.0378 0 0.5>
    Color Minimum=<rgba 10 1.4438 0 0.5>
    Color Mode=::NewEnumerator2
    Lifetime Max=&FN_FrontEndFloatModifier(Gameplay=0.7, Front End=0.4)
    Lifetime Min=0.5
    Lifetime Mode=::NewEnumerator1
    Sprite Size Max=<2 30>
    Sprite Size Min=<1 15>
    Sprite Size Mode=::NewEnumerator2
    UsePositionOffset=false
    Write Material Random=true
   )
   &SkeletalMeshLocation(
    Bone Sampling Mode=::NewEnumerator4
    MaskSpawnGroup=false
    Sampled Position Offset=<0 0 -2>
    Sampled Position Offset Coordinate Space:ENiagaraSkelMeshTransforms=::NewEnumerator2
    Skeletal Mesh=$User.SkeletalMesh
    Use Sampled Mesh Attribute as Ribbon ID=false
   )
   &ShapeLocation(
    Handle Radius=0
    Large Radius=&FN_FrontEndFloatModifier(Gameplay=80, Front End=74.5)
    MaskSpawnGroup=false
    Override Seed=false
    Shape Primitive=::NewEnumerator3
    UseOwnerScale=true
   )
  ]
  particleUpdateScript [
   &ParticleState()
   &SkeletalMeshLocation(
    Bone Sampling Mode=::NewEnumerator4
    "Interpolate Mesh Sampling (Update)"=true
    MaskSpawnGroup=false
    Orientation Sampling=::NewEnumerator1
    Sampled Position Offset Coordinate Space:ENiagaraSkelMeshTransforms=::NewEnumerator0
    Skeletal Mesh=$User.SkeletalMesh
    Use Sampled Mesh Attribute as Ribbon ID=false
   ) {enabled=false, module=SkeletalMeshLocation001}
   &ScaleSpriteSize(
    Non-Uniform Curve Scale=<1 2>
    Non-Uniform Curve Sprite Scale=DataInterface() {propertyValues {
      XCurve {keys [{tangentMode=RCTM_User, time=0, value=1}, {tangentMode=RCTM_User, time=1, value=0.5}]}
      YCurve {keys [{tangentMode=RCTM_User, time=0, value=0.3}, {tangentMode=RCTM_User, time=1, value=1}]}
     }}
    Scale Sprite Size Mode=::NewEnumerator3
   )
   &ScaleColor(
    Scale Alpha=&FloatFromCurve(FloatCurve=DataInterface() {propertyValues {Curve {keys [
         {arriveTangent=5.3877, interpMode=RCIM_Cubic, leaveTangent=5.3877, tangentMode=RCTM_User, time=0, value=0}
         {interpMode=RCIM_Cubic, tangentMode=RCTM_User, time=0.5, value=1}
         {tangentMode=RCTM_User, time=1, value=0}
        ]}}})
    ScaleA=true
    ScaleRGB=false
   )
   &SolveForcesAndVelocity(Clamp Velocity=false, Limit Acceleration=false)
   &CameraOffset(Camera Offset Amount=-10)
  ]
  rendererClasses [&NiagaraSpriteRendererProperties]
  renderers [{rendererClass=&NiagaraSpriteRendererProperties}]
 }
 Magic_Circle_Mesh_B {
  emitterUpdateScript [
   &EmitterState(Life Cycle Mode=::NewEnumerator1, Loop Behavior=::NewEnumerator1, Loop Delay=1.1292, UseLoopDelay=true)
   &SpawnBurst_Instantaneous(Use Spawn Probability=false, UseLoopCountLimit=false)
  ]
  particleSpawnScript [
   &InitializeParticle(
    Color=<rgba 7 0.8857 0.0243 1>
    Mesh Scale=<2.88 2.88 15.6>
    Mesh Scale Mode=::NewEnumerator4
    Sprite Size Mode=::NewEnumerator3
    Uniform Sprite Size=50
    UsePositionOffset=false
    Write Material Random=true
   )
   &SpriteFacingAndAlignment(Sprite Facing=<0 0 1>, WriteAlignment=true, WriteFacing=true)
  ]
  particleUpdateScript [
   &ParticleState(Kill Particles When Lifetime Has Elapsed=false)
   &SkeletalMeshLocation(
    Bone Sampling Mode=::NewEnumerator4
    MaskSpawnGroup=false
    Sampled Position Offset=<0 0 3>
    Sampled Position Offset Coordinate Space:ENiagaraSkelMeshTransforms=::NewEnumerator0
    Skeletal Mesh=$User.SkeletalMesh
    Use Sampled Mesh Attribute as Ribbon ID=false
   )
   &SolveForcesAndVelocity(Clamp Velocity=false, Limit Acceleration=false)
   &DynamicMaterialParameters(
    Index 0 Param 1=3.5
    Index 0 Param 2=1
    Index 0 Param 3=-0.1
    Index 0 Param 4=&FloatFromCurve(
     FloatCurve=DataInterface() {propertyValues {Curve {keys [{arriveTangent=-2.0654, time=0.5, value=1}, {tangentMode=RCTM_User, time=3.5, value=0.05}]}}}
     CurveIndex=$Emitter.Age
     Scale Curve=5
    )
    Index 1 Param 1=0.55
    Index 1 Param 2=0.4
    Index 1 Param 3=3
    Index 2 Param 1=&Multiply_Float(A=$Emitter.Age, B=2)
    Index 2 Param 2=-0.1
    Index 2 Param 3=2
    Module.Param0WriteEnabled=true
    Module.Param0WriteEnabled1=true
    Module.Param0WriteEnabled2=true
    Module.Param1WriteEnabled=true
    Module.Param1WriteEnabled1=true
    Module.Param1WriteEnabled2=false
    Module.Param2WriteEnabled=true
    Module.Param2WriteEnabled1=true
    Module.Param2WriteEnabled2=false
    Module.Param3WriteEnabled=true
    Module.Param3WriteEnabled1=true
    Module.Param3WriteEnabled2=false
    Write Parameter Index 1=true
    Write Parameter Index 2=true
   )
   &UpdateMeshOrientation(Rotation Rate=-0.1, Rotation Vector=<0 0 1>) {enabled=false}
   &ScaleMeshSize(Scale Factor=&VectorFromFloat(Value=&FloatFromCurve(
      FloatCurve=DataInterface() {propertyValues {Curve {keys [
          {interpMode=RCIM_Cubic, tangentMode=RCTM_User, time=1.1, value=0}
          <at 1.4 0.25>
          {arriveTangent=2.7482, interpMode=RCIM_Cubic, leaveTangent=2.7482, tangentMode=RCTM_User, time=1.9, value=0.25}
          {interpMode=RCIM_Cubic, tangentMode=RCTM_User, time=2.4, value=1}
         ]}}}
      CurveIndex=$Emitter.Age
     )))
   &ScaleColor(
    Scale Alpha=&FloatFromCurve(
     FloatCurve=DataInterface() {propertyValues {Curve {keys [{interpMode=RCIM_Cubic, tangentMode=RCTM_User, time=0, value=0}, {interpMode=RCIM_Cubic, tangentMode=RCTM_User, time=0.5, value=1}]}}}
     CurveIndex=$Emitter.Age
    )
    ScaleA=true
    ScaleRGB=false
   )
  ]
  rendererClasses [&NiagaraMeshRendererProperties]
  renderers [{rendererClass=&NiagaraMeshRendererProperties, rendererIndex=0}]
 }
 Magic_Circle_Mesh_G {
  emitterUpdateScript [
   &EmitterState(
    Enable Distance Culling=true
    Life Cycle Mode=::NewEnumerator1
    Loop Behavior=::NewEnumerator1
    Loop Delay=1.1292
    Loop Duration Mode=::NewEnumerator1
    MaxDistance=8000
    Module.bUseMaxDistance=true
    Module.bUseMinDistance=false
    Scalability Mode=::NewEnumerator1
    UseLoopDelay=true
   )
   &SpawnBurst_Instantaneous(Use Spawn Probability=false, UseLoopCountLimit=false)
   &SpawnBurst_Instantaneous(Spawn Time=1.5, Use Spawn Probability=false, UseLoopCountLimit=false) {module=SpawnBurst_Instantaneous001}
  ]
  particleSpawnScript [
   &self:ParticleID() {module=ParticleID}
   &InitializeParticle(
    Color=<rgba 7 0.8857 0.0243 1>
    Mesh Scale=<2.88 2.88 15.6>
    Mesh Scale Mode=::NewEnumerator4
    Sprite Size Mode=::NewEnumerator3
    Uniform Sprite Size=50
    UsePositionOffset=false
    Write Material Random=true
   )
   &SpriteFacingAndAlignment(Sprite Facing=<0 0 1>, WriteAlignment=true, WriteFacing=true)
  ]
  particleUpdateScript [
   &ParticleState(Kill Particles When Lifetime Has Elapsed=false)
   &SkeletalMeshLocation(
    Bone Sampling Mode=::NewEnumerator4
    "Interpolate Mesh Sampling (Update)"=true
    MaskSpawnGroup=false
    Sampled Orientation Coordinate Space=::Local
    Sampled Position Offset=<0 0 3>
    Sampled Position Offset Coordinate Space:ENiagaraSkelMeshTransforms=::NewEnumerator0
    Skeletal Mesh=$User.SkeletalMesh
    Use Sampled Mesh Attribute as Ribbon ID=false
   )
   &UpdateMeshOrientation(
    Rotation Coordinate Space=::NewEnumerator1
    Rotation Rate=&FloatFromCurve(FloatCurve=DataInterface() {propertyValues {Curve {keys [<at 2.3 4>,<at 3.4 1>]}}}, CurveIndex=$Emitter.Age, Scale Curve=-0.1)
    Rotation Vector=<0 0 1>
   ) {enabled=false}
   &SolveForcesAndVelocity(Clamp Velocity=false, Limit Acceleration=false)
   &DynamicMaterialParameters(
    Index 0 Param 1=3.5
    Index 0 Param 2=1
    Index 0 Param 3=-0.1
    Index 1 Param 1=3
    Index 1 Param 2=3
    Index 1 Param 3=3
    Index 2 Param 1=HlslExpression() {hlslExpression=Particles.Age * 2.0 +5.4}
    Index 2 Param 2=-0.1
    Index 2 Param 3=2
    Module.Param0WriteEnabled=true
    Module.Param0WriteEnabled1=true
    Module.Param0WriteEnabled2=true
    Module.Param1WriteEnabled=true
    Module.Param1WriteEnabled1=true
    Module.Param1WriteEnabled2=false
    Module.Param2WriteEnabled=true
    Module.Param2WriteEnabled1=true
    Module.Param2WriteEnabled2=false
    Module.Param3WriteEnabled=true
    Module.Param3WriteEnabled1=true
    Module.Param3WriteEnabled2=false
    Write Parameter Index 1=true
    Write Parameter Index 2=true
   )
   &ScaleMeshSize(Scale Factor=&VectorFromFloat(Value=&FloatFromCurve(
      FloatCurve=DataInterface() {propertyValues {Curve {keys [
          {interpMode=RCIM_Cubic, tangentMode=RCTM_User, time=1.1, value=0.001}
          <at 1.4 0.25>
          {arriveTangent=2.7482, interpMode=RCIM_Cubic, leaveTangent=2.7482, tangentMode=RCTM_User, time=1.9, value=0.25}
          {interpMode=RCIM_Cubic, tangentMode=RCTM_User, time=2.4, value=1}
         ]}}}
      CurveIndex=$Emitter.Age
     )))
   &ScaleMeshSize(Scale Factor=&VectorFromCurve(
     VectorCurve=DataInterface() {propertyValues {XCurve {keys [<at 2 0 RCIM_Cubic>,<at 3 1 RCIM_Cubic>]}, YCurve {keys [<at 0 1>,<at 1 1>]}, ZCurve {keys [<at 0 1>,<at 1 1>]}}}
     CurveIndex=&Add_Float(A=$Emitter.Age, B=$Particles.PID)
    )) {module=ScaleMeshSize001}
   &ScaleColor(
    Scale Alpha=&FloatFromCurve(
     FloatCurve=DataInterface() {propertyValues {Curve {keys [{interpMode=RCIM_Cubic, tangentMode=RCTM_User, time=2, value=0}, {interpMode=RCIM_Cubic, tangentMode=RCTM_User, time=2.5, value=1}]}}}
     CurveIndex=$Emitter.Age
    )
    ScaleA=true
    ScaleRGB=false
   )
  ]
  rendererClasses [&NiagaraMeshRendererProperties]
  renderers [{rendererClass=&NiagaraMeshRendererProperties, rendererIndex=0}]
 }
 Magic_Circle_Sparks {
  emitterUpdateScript [
   &EmitterState(Life Cycle Mode=::NewEnumerator1, Loop Behavior=::NewEnumerator1, Loop Delay=1.1292, UseLoopDelay=false)
   &SpawnBurst_Instantaneous(Spawn Time=1.1292, Use Spawn Probability=false, UseLoopCountLimit=false)
  ]
  particleSpawnScript [
   &InitializeParticle(
    Color=<rgba 6 0.9364 0.1979 0.8>
    Mesh Scale=<1.248 1.248 3.6>
    Mesh Scale Mode=::NewEnumerator4
    Sprite Size Mode=::NewEnumerator3
    Uniform Sprite Size=270
    UsePositionOffset=false
    Write Material Random=true
   ) {module=InitializeParticle001}
   &SpriteFacingAndAlignment(Sprite Facing=<0 0 1>, WriteAlignment=true, WriteFacing=true) {enabled=false}
   /SetVariables(Particles.Position=HlslExpression() {hlslExpression="float3(Particles.Position.xy, -5.0) /* Custom HLSL! */"}) {isSetParameters=true, module=SetVariables_#1}
  ]
  particleUpdateScript [
   &ParticleState(Kill Particles When Lifetime Has Elapsed=false)
   &AlignSpriteToMeshOrientation(Mesh Orientation Relative Sprite Alignment Vector=<1 0 0>, Mesh Orientation Relative Sprite Facing Vector:Vector3f=<0 0 1>)
   &SkeletalMeshLocation(
    Bone Sampling Mode=::NewEnumerator4
    MaskSpawnGroup=false
    Sampled Position Offset Coordinate Space:ENiagaraSkelMeshTransforms=::NewEnumerator0
    Skeletal Mesh=$User.SkeletalMesh
    Use Sampled Mesh Attribute as Ribbon ID=false
   )
   &SolveForcesAndVelocity(Clamp Velocity=false, Limit Acceleration=false)
   &DynamicMaterialParameters(
    Index 0 Param 1=2
    Index 0 Param 2=4
    Index 1 Param 4=0.3
    Index 2 Param 1=&Multiply_Float(A=$Emitter.Age, B=1.4)
    Module.Param0WriteEnabled=true
    Module.Param0WriteEnabled1=false
    Module.Param0WriteEnabled2=true
    Module.Param1WriteEnabled=true
    Module.Param1WriteEnabled1=false
    Module.Param1WriteEnabled2=false
    Module.Param2WriteEnabled=true
    Module.Param2WriteEnabled1=false
    Module.Param2WriteEnabled2=false
    Module.Param3WriteEnabled=true
    Module.Param3WriteEnabled1=true
    Module.Param3WriteEnabled2=false
    Write Parameter Index 1=true
    Write Parameter Index 2=true
   )
   &UpdateMeshOrientation(Rotation Rate=-0.1, Rotation Vector=<0 0 1>) {enabled=false}
   &ScaleMeshSize(Scale Factor=&VectorFromFloat(Value=&FloatFromCurve(
      FloatCurve=DataInterface() {propertyValues {Curve {keys [
          {interpMode=RCIM_Cubic, tangentMode=RCTM_User, time=1.1, value=0}
          <at 1.4 0.25>
          {arriveTangent=2.7482, interpMode=RCIM_Cubic, leaveTangent=2.7482, tangentMode=RCTM_User, time=1.9, value=0.25}
          {interpMode=RCIM_Cubic, tangentMode=RCTM_User, time=2.4, value=1}
         ]}}}
      CurveIndex=$Emitter.Age
     )))
   &ScaleSpriteSize(
    Uniform Curve Index=$Emitter.Age
    Uniform Curve Sprite Scale=DataInterface() {propertyValues {Curve {keys [
        {interpMode=RCIM_Cubic, tangentMode=RCTM_User, time=1.1, value=0}
        <at 1.4 0.25>
        {arriveTangent=2.7482, interpMode=RCIM_Cubic, leaveTangent=2.7482, tangentMode=RCTM_User, time=1.9, value=0.25}
        {interpMode=RCIM_Cubic, tangentMode=RCTM_User, time=2.4, value=1}
       ]}}}
   )
   &ScaleColor(
    Scale Alpha=&FloatFromCurve(
     FloatCurve=DataInterface() {propertyValues {Curve {keys [{interpMode=RCIM_Cubic, tangentMode=RCTM_User, time=1.5, value=0}, {interpMode=RCIM_Cubic, tangentMode=RCTM_User, time=2.5, value=1}]}}}
     CurveIndex=$Emitter.Age
    )
    ScaleA=true
    ScaleRGB=false
   )
  ]
  rendererClasses [&NiagaraMeshRendererProperties, &NiagaraSpriteRendererProperties]
  renderers [
   {rendererClass=&NiagaraMeshRendererProperties, rendererIndex=0}
   {rendererClass=&NiagaraSpriteRendererProperties, rendererIndex=1}
  ]
 }
 Magic_OpaqueRing {
  emitterUpdateScript [
   &EmitterState(
    Enable Distance Culling=true
    Life Cycle Mode=::NewEnumerator1
    Loop Behavior=::NewEnumerator1
    Loop Delay=1.1292
    MaxDistance=7000
    Module.bUseMaxDistance=true
    Module.bUseMinDistance=false
    Scalability Mode=::NewEnumerator1
    UseLoopDelay=false
   )
   &SpawnBurst_Instantaneous(Spawn Time=1.1292, Use Spawn Probability=false, UseLoopCountLimit=false)
  ]
  particleSpawnScript [
   &InitializeParticle(
    Color=<rgba 5.5 1.375 0.5061 1>
    Mesh Scale=<2.7 2.7 15>
    Mesh Scale Mode=::NewEnumerator4
    Sprite Size Mode=::NewEnumerator3
    Uniform Sprite Size=50
    UsePositionOffset=false
    Write Material Random=true
   )
   &SpriteFacingAndAlignment(Sprite Facing=<0 0 1>, WriteAlignment=true, WriteFacing=true) {enabled=false}
   /SetVariables(Particles.Position=HlslExpression() {hlslExpression="float3(Particles.Position.xy, -5.0) /* Custom HLSL! */"}) {isSetParameters=true, module=SetVariables_#1}
  ]
  particleUpdateScript [
   &ParticleState(Kill Particles When Lifetime Has Elapsed=false)
   &SkeletalMeshLocation(
    Bone Sampling Mode=::NewEnumerator4
    MaskSpawnGroup=false
    Sampled Position Offset=<0 0 2>
    Sampled Position Offset Coordinate Space:ENiagaraSkelMeshTransforms=::NewEnumerator0
    Skeletal Mesh=$User.SkeletalMesh
    Use Sampled Mesh Attribute as Ribbon ID=false
   )
   &SolveForcesAndVelocity(Clamp Velocity=false, Limit Acceleration=false)
   &DynamicMaterialParameters(
    Index 0 Param 1=3.5
    Index 0 Param 2=1
    Index 0 Param 3=-0.1
    Index 0 Param 4=&FloatFromCurve(
     FloatCurve=DataInterface() {propertyValues {Curve {keys [{arriveTangent=-2.0654, time=1.25, value=0.5}, {tangentMode=RCTM_User, time=5.2, value=0.05}]}}}
     CurveIndex=$Emitter.Age
    )
    Index 1 Param 1=0.55
    Index 1 Param 2=0.4
    Index 1 Param 3=0.5
    Module.Param0WriteEnabled=true
    Module.Param0WriteEnabled1=true
    Module.Param1WriteEnabled=true
    Module.Param1WriteEnabled1=true
    Module.Param2WriteEnabled=true
    Module.Param2WriteEnabled1=true
    Module.Param3WriteEnabled=true
    Module.Param3WriteEnabled1=true
    Write Parameter Index 1=true
   )
   &UpdateMeshOrientation(Rotation Rate=-0.1, Rotation Vector=<0 0 1>) {enabled=false}
   &ScaleMeshSize(Scale Factor=&VectorFromFloat(Value=&FloatFromCurve(
      FloatCurve=DataInterface() {propertyValues {Curve {keys [
          {interpMode=RCIM_Cubic, tangentMode=RCTM_User, time=1.1, value=0}
          <at 1.4 0.25>
          {arriveTangent=2.7482, interpMode=RCIM_Cubic, leaveTangent=2.7482, tangentMode=RCTM_User, time=1.9, value=0.25}
          {interpMode=RCIM_Cubic, tangentMode=RCTM_User, time=2.4, value=1}
         ]}}}
      CurveIndex=$Emitter.Age
     )))
   &ScaleColor(
    Scale Alpha=&FloatFromCurve(
     FloatCurve=DataInterface() {propertyValues {Curve {keys [{interpMode=RCIM_Cubic, tangentMode=RCTM_User, time=1.5, value=0}, {interpMode=RCIM_Cubic, tangentMode=RCTM_User, time=2.5, value=1}]}}}
     CurveIndex=$Emitter.Age
    )
    ScaleA=true
    ScaleRGB=false
   )
   &FN_Light_Attributes(
    Color=<rgba 1 0.2 0 1>
    Create Color Attribute=true
    Create Exponent Attribute=true
    Create Radius Attribute=true
    Exponent=2
    Radius=90
    Scale Color=&FloatFromCurve(
     FloatCurve=DataInterface() {propertyValues {Curve {keys [
         {interpMode=RCIM_Cubic, tangentMode=RCTM_User, time=1.1, value=0}
         <at 1.4 0.25>
         {arriveTangent=2.7482, interpMode=RCIM_Cubic, leaveTangent=2.7482, tangentMode=RCTM_User, time=1.9, value=0.25}
         {interpMode=RCIM_Cubic, tangentMode=RCTM_User, time=2.4, value=1}
        ]}}}
     CurveIndex=$Emitter.Age
     Scale Curve=1.29
    )
   )
  ]
  rendererClasses [&NiagaraMeshRendererProperties, &NiagaraLightRendererProperties]
  renderers [
   {rendererClass=&NiagaraMeshRendererProperties, rendererIndex=0}
   {rendererClass=&NiagaraLightRendererProperties, rendererIndex=1}
  ]
 }
 Magic_SparksCircle_Mesh {
  bEnabled=false
  emitterUpdateScript [
   &EmitterState(Life Cycle Mode=::NewEnumerator1, Loop Behavior=::NewEnumerator1, Loop Duration=8.3884, UseLoopDelay=false)
   &SpawnBurst_Instantaneous(Use Spawn Probability=false, UseLoopCountLimit=false)
   &SpawnBurst_Instantaneous(Spawn Time=0.4863, Use Spawn Probability=false, UseLoopCountLimit=false) {module=SpawnBurst_Instantaneous001}
  ]
  particleSpawnScript [
   &InitializeParticle(
    Color=<rgba 6 0.9364 0.1979 0.8>
    Mesh Scale=<1.248 1.248 3.6>
    Mesh Scale Mode=::NewEnumerator4
    Sprite Size Mode=::NewEnumerator3
    Uniform Sprite Size=270
    UsePositionOffset=false
    Write Material Random=true
   )
   &SpriteFacingAndAlignment(
    Alignment Coordinate Space=::World
    Facing Coordinate Space=::World
    Sprite Alignment=<0 0 1>
    Sprite Facing=<0 0 1>
    WriteAlignment=true
    WriteFacing=true
   ) {enabled=false}
   &AddRotationalVelocity(Coordinate Space:ENiagaraMeshTransforms=::NewEnumerator1, Rotation Rate=<0 0 1>)
   /SetVariables(Particles.Position=HlslExpression() {hlslExpression="float3(Particles.Position.xy, -5.0) /* Custom HLSL! */"}) {isSetParameters=true}
  ]
  particleUpdateScript [
   &ParticleState(Kill Particles When Lifetime Has Elapsed=false)
   &AlignSpriteToMeshOrientation(Mesh Orientation Relative Sprite Alignment Vector=<1 0 0>, Mesh Orientation Relative Sprite Facing Vector:Vector3f=<0 0 1>)
   &SkeletalMeshLocation(
    Bone Sampling Mode=::NewEnumerator4
    MaskSpawnGroup=false
    Sampled Position Offset Coordinate Space:ENiagaraSkelMeshTransforms=::NewEnumerator0
    Skeletal Mesh=$User.SkeletalMesh
    Use Sampled Mesh Attribute as Ribbon ID=false
   )
   &DynamicMaterialParameters(
    Index 0 Param 1=2
    Index 0 Param 2=4
    Index 1 Param 4=0.3
    Index 2 Param 1=&Multiply_Float(A=$Emitter.Age, B=1.1)
    Module.Param0WriteEnabled=true
    Module.Param0WriteEnabled1=false
    Module.Param0WriteEnabled2=true
    Module.Param1WriteEnabled=true
    Module.Param1WriteEnabled1=false
    Module.Param1WriteEnabled2=false
    Module.Param2WriteEnabled=true
    Module.Param2WriteEnabled1=false
    Module.Param2WriteEnabled2=false
    Module.Param3WriteEnabled=true
    Module.Param3WriteEnabled1=true
    Module.Param3WriteEnabled2=false
    Write Parameter Index 1=true
    Write Parameter Index 2=true
   )
   &SolveForcesAndVelocity(Clamp Velocity=false, Limit Acceleration=false)
   &SolveRotationalForcesAndVelocity()
   &UpdateMeshOrientation(Rotation Coordinate Space=::NewEnumerator1, Rotation Rate=-1.2, Rotation Vector=<0 0 1>) {enabled=false}
   &ScaleMeshSize(Scale Factor=&VectorFromFloat(Value=&FloatFromCurve(
      FloatCurve=DataInterface() {propertyValues {Curve {keys [
          {interpMode=RCIM_Cubic, tangentMode=RCTM_User, time=1.1, value=0}
          <at 1.4 0.24>
          {arriveTangent=2.7482, interpMode=RCIM_Cubic, leaveTangent=2.7482, tangentMode=RCTM_User, time=1.9, value=0.25}
          {interpMode=RCIM_Cubic, tangentMode=RCTM_User, time=2.4, value=1}
         ]}}}
      CurveIndex=$Emitter.Age
     )))
   &ScaleSpriteSize(
    Uniform Curve Index=$Emitter.Age
    Uniform Curve Sprite Scale=DataInterface() {propertyValues {Curve {keys [
        {interpMode=RCIM_Cubic, tangentMode=RCTM_User, time=1.1, value=0}
        <at 1.4 0.25>
        {arriveTangent=2.7482, interpMode=RCIM_Cubic, leaveTangent=2.7482, tangentMode=RCTM_User, time=1.9, value=0.25}
        {interpMode=RCIM_Cubic, tangentMode=RCTM_User, time=2.4, value=1}
       ]}}}
   )
   &ScaleColor(
    Scale Alpha=&FloatFromCurve(
     FloatCurve=DataInterface() {propertyValues {Curve {keys [{interpMode=RCIM_Cubic, tangentMode=RCTM_User, time=1.4, value=0}, {interpMode=RCIM_Cubic, tangentMode=RCTM_User, time=2, value=1}]}}}
     CurveIndex=$Emitter.Age
    )
    ScaleA=true
    ScaleRGB=false
   )
  ]
  rendererClasses [&NiagaraMeshRendererProperties, &NiagaraSpriteRendererProperties]
  renderers [
   {rendererClass=&NiagaraMeshRendererProperties, rendererIndex=0}
   {rendererClass=&NiagaraSpriteRendererProperties, rendererIndex=1}
   {rendererClass=&NiagaraMeshRendererProperties, rendererIndex=2}
  ]
 }
}
summary {
 systemName=NS_Glider_Mystic_M_Idle
 emitters [
  {emitterName=Magic_Circle_Mesh_R, rendererClasses [&NiagaraMeshRendererProperties]}
  {emitterName=GPU_Sparks_gameplay_, simTarget=GPUComputeSim, rendererClasses [&NiagaraSpriteRendererProperties]}
  {
   bEnabled=false
   emitterName=Magic_SparksCircle_Mesh
   rendererClasses [&NiagaraMeshRendererProperties, &NiagaraSpriteRendererProperties]
  }
  {emitterName=LineSprite, rendererClasses [&NiagaraSpriteRendererProperties]}
  {emitterName=Center_Glow, rendererClasses [&NiagaraSpriteRendererProperties]}
  {emitterName=GPU_Sparks_Tail, simTarget=GPUComputeSim, rendererClasses [&NiagaraSpriteRendererProperties]}
  {emitterName=Magic_Circle_Mesh_G, rendererClasses [&NiagaraMeshRendererProperties]}
  {emitterName=First_Sparks, simTarget=GPUComputeSim, rendererClasses [&NiagaraSpriteRendererProperties]}
  {emitterName=FN_Ribbon_Attributes_Test, rendererClasses [&NiagaraRibbonRendererProperties]}
  {emitterName=GPU_Sparks_Init, simTarget=GPUComputeSim, rendererClasses [&NiagaraSpriteRendererProperties]}
  {emitterName=Magic_Circle_Mesh_B, rendererClasses [&NiagaraMeshRendererProperties]}
  {
   emitterName=Magic_OpaqueRing
   rendererClasses [&NiagaraMeshRendererProperties, &NiagaraLightRendererProperties]
  }
  {emitterName=Feet_Glow, rendererClasses [&NiagaraSpriteRendererProperties]}
  {
   emitterName=Magic_Circle_Sparks
   rendererClasses [&NiagaraMeshRendererProperties, &NiagaraSpriteRendererProperties]
  }
  {emitterName=GPU_Sparks_Frontend, simTarget=GPUComputeSim, rendererClasses [&NiagaraSpriteRendererProperties]}
 ]
 userVariables [
  User.BaseColor {type=LinearColor, description="", defaultValue {struct=LinearColor, value=<rgba 5 1.3741 0 0.1>}}
  User.bIsFrontEnd {type=NiagaraBool}
  User.bIsFrontEndPreview {type=NiagaraBool}
  User.ForwardDot {type=NiagaraFloat, defaultValue=1, description=""}
  User.RibbonIDOffset {type=NiagaraInt32}
  User.SkeletalMesh {
   type=NiagaraDataInterfaceSkeletalMesh
   defaultValue {
    struct=NiagaraExt_VariableValue_DataInterface
    value {dataInterface=&self:NiagaraDataInterfaceSkeletalMesh_0}
   }
  }
 ]
}
